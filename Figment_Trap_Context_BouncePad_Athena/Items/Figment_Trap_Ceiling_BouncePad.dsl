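; Ceiling bounce pad trap that launches overlapping players, vehicles and physics props along its arrow.
; Its trigger, enabled once built or previewed, routes actors to type-specific launches; player and vehicle launches mirror and clamp incoming velocity.
; Those two launches skip the next fall damage and apply low-gravity and bounce-tracker effects to the stored player; a pawn with the max-count tag destroys the trap instead.
; An activation GameplayCue handles sound and FX.
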(event EditorPreview|EventSetupForEditorPreview
  ;[0.5917 1 0 0.5] Ensure the trigger is set up when placed in UEFN during VK-Edit
  (|CallParentFunction)
  (Utilities|FlowControl|DoOnce false
    (:Completed
      (Collision|SetCollisionEnabled (Variables|BounceTrap|GetTrigger) "QueryAndPhysics")
      ;[1 1 1] Initial Setup
      (Utilities|FlowControl|DoOnce false
        (:Completed
          (Trap|AddTriggerComponent (Variables|BounceTrap|GetTrigger))
          (if (not (Editor|InEditor))
            (Audio|PlaySoundatLocation (Variables|Audio|GetPlace_Trap_Sound) (Transformation|GetActorLocation))
            (Utilities|TrapPlacementDustFX :Attach%20to%20Component (Variables|Debug|GetStaticMeshComponent) :Socket "PlacedFX" :Rotation "0, 0, 0" :Burst%20Rate 15 :Size%20%28Volume%29 "200.000000,200.000000,15.000000" :Velocity%20Scale "1,1,1" :Size%20Scale%20%28Particles%29 1
              (:Out
                _)))
          (if (> (ScalableFloat|GetValueatLevel (Variables|Default|GetSetIgnoreFallDamage)) 0.0)
            (Variables|Default|SetIgnoreFallDamage true)
            (else
              (Variables|Default|SetIgnoreFallDamage false))))))) ; Otherwise this happens whenever it's moved in the editor
)

(event EventOnPlaced
  ;[1 1 1] Initial Setup
  (Utilities|FlowControl|DoOnce false
    (:Completed
      (Trap|AddTriggerComponent (Variables|BounceTrap|GetTrigger))
      (if (not (Editor|InEditor))
        (Audio|PlaySoundatLocation (Variables|Audio|GetPlace_Trap_Sound) (Transformation|GetActorLocation))
        (Utilities|TrapPlacementDustFX :Attach%20to%20Component (Variables|Debug|GetStaticMeshComponent) :Socket "PlacedFX" :Rotation "0, 0, 0" :Burst%20Rate 15 :Size%20%28Volume%29 "200.000000,200.000000,15.000000" :Velocity%20Scale "1,1,1" :Size%20Scale%20%28Particles%29 1
          (:Out
            _)))
      (if (> (ScalableFloat|GetValueatLevel (Variables|Default|GetSetIgnoreFallDamage)) 0.0)
        (Variables|Default|SetIgnoreFallDamage true)
        (else
          (Variables|Default|SetIgnoreFallDamage false))))))

(event EventOnFinishedBuilding
  ;[1 1 1] Finished building
  (Collision|SetCollisionEnabled (Variables|BounceTrap|GetTrigger) "QueryAndPhysics")
  (Networking|CanExecuteCosmeticEvents
    (:True
      (if (not (Editor|InEditor))
        (Audio|PlaySoundatLocation (Variables|Audio|GetTrap_Active_Sound) (Transformation|GetActorLocation))))))

(event OnComponentBeginOverlap(Trigger)
  (params (OverlappedComponent :type Primitive%20Component%20Object%20Reference) (OtherActor :type Actor%20Object%20Reference) (OtherComp :type Primitive%20Component%20Object%20Reference) (OtherBodyIndex :type Integer) (bFromSweep :type Boolean) (SweepResult :type Hit%20Result%20Structure%20%28by%20ref%29))
  ;[1 1 1] On Overlap
  (bind _output_get (Variables|BounceTrap|SetOverlappedactor OtherActor))
  (bind _asfort_player_pawn_athena (Utilities|Casting|CastToFortPlayerPawnAthena _output_get)
    (:then
      (bind _output_get_2 (Variables|Default|SetLaunchedPlayer _asfort_player_pawn_athena))
      (if (GameplayTags|HasMatchingGameplayTag _output_get_2 (Variables|Default|GetBouncePadTagMaxCount))
        (Actor|DestroyActor)
        (elif (Vehicle|IsinVehicle _asfort_player_pawn_athena)
          (CallFunction|VehicleLaunch (Vehicle|GetVehicleInterface _asfort_player_pawn_athena))
          (else
            (CallFunction|PlayerLaunch _asfort_player_pawn_athena)))))
    (:CastFailed
      (bind _asfort_vehicle_interface (Utilities|Casting|CastToFortVehicleInterface (Variables|BounceTrap|GetOverlappedactor))
        (:then
          (bind _output_get_3 (Variables|Default|SetLaunchedVehicleInterface _asfort_vehicle_interface))
          (CallFunction|VehicleLaunch _output_get_3))
        (:CastFailed
          (CallFunction|PhysicsPropLaunch (Variables|BounceTrap|GetOverlappedactor)))))))

(event GameplayCue|GameplayCue.Abilities.Activation.Traps.ActivateTrap
  (params (EventType :type EGameplayCueEvent%20Enum :default "OnActive") (Parameters :type Gameplay%20Cue%20Parameters%20Structure%20%28by%20ref%29))
  (bind _overlappedactor (Variables|BounceTrap|GetOverlappedactor))
  ;[1 1 1] GC Triggered on Overlap
  (switch Utilities|FlowControl|Switch|SwitchonEGameplayCueEvent EventType
    (:Executed
      ;;[1 1 1] Check if actor is a pawn or vehicle
      (bind _asfort_player_pawn_athena (Utilities|Casting|CastToFortPlayerPawnAthena _overlappedactor)
        (:then
          _)
        (:CastFailed
          (bind _asfort_athena_skvehicle (Utilities|Casting|CastToFortAthenaSKVehicle _overlappedactor)
            (:then
              (Effects|SpawnEmitterAttached "/Game/Weapons/FORT_TRAPS/Trap_Jump_FullTile/Effects/P_Trap_Jump_FullTile_VehicleTrail.P_Trap_Jump_FullTile_VehicleTrail" (Variables|/Script/FortniteGame.FortAthenaSKVehicle|GetSkeletalMesh _asfort_athena_skvehicle) ""))
            (:CastFailed
              (Development|PrintString "Overlap was not a  pawn or vehicle")))))
      (Audio|PlaySoundatLocation (Variables|Audio|GetTrap_Fire_Sound) (Transformation|GetActorLocation))
      (Components|Activation|Activate (Variables|Default|GetBouncePadFX) true))))

(event EventDestroyed
  ;[1 1 1] Trap Destroyed
  (if (not (Editor|InEditor))
    (Audio|PlaySoundatLocation (Variables|Audio|GetTrap_Destroyed_Sound) (Transformation|GetActorLocation))))

(fn ApplyGameplayEffects
  (bind _self self)
  (bind _returnvalue (Ability|GetAbilitySystemComponent _self))
  (bind _launchedplayer (Variables|Default|GetLaunchedPlayer))
  (bind _returnvalue_2 (Ability|GetAbilitySystemComponent _launchedplayer))
  (if (Variables|Default|GetIgnoreFallDamage)
    ;[1 1 1] Player Pawn
    (GameplayEffects|ApplyGameplayEffectToTarget :self _returnvalue :GameplayEffectClass (Variables|Default|GetGE_LowGravity) :Target _returnvalue_2 :Level 0.0)
    (else
      (GameplayEffects|ApplyGameplayEffectToTarget :self _returnvalue :GameplayEffectClass (Variables|Default|GetGE_LowGravityNoFX) :Target _returnvalue_2 :Level 0.0)))
  (GameplayEffects|ApplyGameplayEffectToTarget :self _returnvalue :GameplayEffectClass (Variables|Default|GetGE_BounceTracker1) :Target _returnvalue_2 :Level 0.0)
  (GameplayEffects|ApplyGameplayEffectToTarget :self _returnvalue :GameplayEffectClass (Variables|Default|GetGE_BounceTracker2) :Target _returnvalue_2 :Level 0.0)
  (GameplayEffects|ApplyGameplayEffectToTarget :self _returnvalue :GameplayEffectClass (Variables|Default|GetGE_LowGravityWithHopRocks) :Target _returnvalue_2 :Level 0.0)
  (Variables|Default|SetLaunchedPlayer 0))

(fn OnRep_Overlappedactor)

(fn PhysicsPropLaunch
  (params (PhysicsPropToLaunch :type Actor%20Object%20Reference))
  (Class|NewtonPhysicsPropsLibrary|LaunchPhysicsProps self PhysicsPropToLaunch (Transformation|GetForwardVector (Variables|Default|GetArrow)) (Variables|Default|GetMinPhysicsPropMass) (Variables|Default|GetMaxPhysicsPropMass) (Variables|Default|GetMinPhysicsPropImpulse) (Variables|Default|GetMaxPhysicsPropImpulse))
  (return))

(fn PlayerLaunch
  (params (Player :type Fort%20Player%20Pawn%20Athena%20Object%20Reference))
  (Character|LaunchCharacter Player (Math|Vector|ClampVectorSize (+ (Math|Vector|ClampVectorSize (Math|Vector|MakeVector (.x (Math|Vector|MirrorVectorbyNormal (Transformation|GetVelocity Player) (Transformation|GetForwardVector (Variables|Default|GetArrow)))) (.y (Math|Vector|MirrorVectorbyNormal (Transformation|GetVelocity Player) (Transformation|GetForwardVector (Variables|Default|GetArrow))))) 0.0 (ScalableFloat|GetValueatLevel (Variables|Default|GetPlayerMaxLateralVelocity))) (* (Transformation|GetForwardVector (Variables|Default|GetArrow)) (ScalableFloat|GetValueatLevel (Variables|Default|GetPlayerMaxDownwardVelocity)))) 5e+02 (ScalableFloat|GetValueatLevel (Variables|Default|GetPlayerMaxVelocity))) true true)
  (Variables|/Script/FortniteGame.FortPawn|SetbIgnoreNextFallingDamage true Player)
  (CallFunction|ApplyGameplayEffects))

(fn ConstructionScript
  (Collision|SetCollisionEnabled (Variables|BounceTrap|GetTrigger) "NoCollision"))

(fn VehicleLaunch
  (params (Vehicle :type Fort%20Vehicle%20Interface%20Interface))
  (bind _returnvalue (Vehicle|GetVehicleActor Vehicle :!class /Script/FortniteGame.FortVehicleInterface))
  (bind _asskeletal_mesh_component (Utilities|Casting|CastToSkeletalMeshComponent (Variables|/Script/Engine.Actor|GetRootComponent _returnvalue)))
  (Audio|BroadcastAbilityImpactatLocation (Variables|Audio|GetTrap_Active_Sound) _returnvalue (Math|Vector|ClampVectorSize (+ (Math|Vector|MakeVector (.x (+ (Math|Vector|MirrorVectorbyNormal (Vehicle|GetVehicleVelocity Vehicle) (Transformation|GetForwardVector (Variables|Default|GetArrow))) (* (Transformation|GetForwardVector (Variables|Default|GetArrow)) (* (Physics|GetMass _asskeletal_mesh_component) (ScalableFloat|GetValueatLevel (Variables|Default|GetVehicleVelocityMultiplier)))))) (.y (+ (Math|Vector|MirrorVectorbyNormal (Vehicle|GetVehicleVelocity Vehicle) (Transformation|GetForwardVector (Variables|Default|GetArrow))) (* (Transformation|GetForwardVector (Variables|Default|GetArrow)) (* (Physics|GetMass _asskeletal_mesh_component) (ScalableFloat|GetValueatLevel (Variables|Default|GetVehicleVelocityMultiplier))))))) (* (Transformation|GetForwardVector (Variables|Default|GetArrow)) (ScalableFloat|GetValueatLevel (Variables|Default|GetVehicleMinLateralVelocity)))) (ScalableFloat|GetValueatLevel (Variables|Default|GetVehicleMinLateralVelocity)) (ScalableFloat|GetValueatLevel (Variables|Default|GetVehicleMaxVelocity))) 7.8e+02) ; Set Physics Linear Velocity
  (Vehicle|SetIgnoreNextFallingDamage Vehicle true)
  (CallFunction|ApplyGameplayEffects))
