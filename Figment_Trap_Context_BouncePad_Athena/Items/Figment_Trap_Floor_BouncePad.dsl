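; Floor bounce pad trap that launches players, vehicles, AI pawns and physics objects or props that touch it.
; Trigger and actor-overlap events feed HandleActorLaunch, with a briefly held recently-launched list so the overlap fallback avoids double launches.
; It casts by type to launch functions, most combining pad direction with clamped, mirrored incoming velocity.
; Pawn launches apply low-gravity and bounce-tracker effects plus optional fall-damage immunity; a gameplay cue drives fire effects.
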(event EventBeginPlay
  (|CallParentFunction)
  (Collision|SetCollisionEnabled (Variables|BouncePad|GetTrigger) "QueryAndPhysics"))

(event EditorPreview|EventSetupForEditorPreview
  ;[0.5917 1 0 0.5] Ensure the trigger is set up when placed in UEFN during VK-Edit
  (|CallParentFunction)
  (Utilities|FlowControl|DoOnce false
    (:Completed
      (Utilities|FlowControl|Sequence
        (:then_0
          (Collision|SetCollisionEnabled (Variables|BouncePad|GetTrigger) "QueryAndPhysics"))
        (:then_1
          ;[1 1 1] Initial Setup
          (Utilities|FlowControl|DoOnce false
            (:Completed
              (Trap|AddTriggerComponent (Variables|BouncePad|GetTrigger))
              (if (not (Editor|InEditor))
                (Audio|PlaySoundatLocation (Variables|Audio|GetPlace_Trap_Sound) (Transformation|GetActorLocation))
                (Utilities|TrapPlacementDustFX :Attach%20to%20Component (Variables|Debug|GetStaticMeshComponent) :Socket "PlacedFX" :Rotation "0, 0, 0" :Burst%20Rate 15 :Size%20%28Volume%29 "200.000000,200.000000,15.000000" :Velocity%20Scale "1,1,1" :Size%20Scale%20%28Particles%29 1
                  (:Out
                    _)))
              (if (> (ScalableFloat|GetValueatLevel (Variables|BouncePad|GetSetIgnoreFallDamage)) 0.0)
                (Variables|BouncePad|SetIgnoreFallDamage true)
                (else
                  (Variables|BouncePad|SetIgnoreFallDamage false))))))))) ; Otherwise this happens whenever it's moved in the editor
)

(event EventOnPlaced
  ;[1 1 1] Initial Setup
  (Utilities|FlowControl|DoOnce false
    (:Completed
      (Trap|AddTriggerComponent (Variables|BouncePad|GetTrigger))
      (if (not (Editor|InEditor))
        (Audio|PlaySoundatLocation (Variables|Audio|GetPlace_Trap_Sound) (Transformation|GetActorLocation))
        (Utilities|TrapPlacementDustFX :Attach%20to%20Component (Variables|Debug|GetStaticMeshComponent) :Socket "PlacedFX" :Rotation "0, 0, 0" :Burst%20Rate 15 :Size%20%28Volume%29 "200.000000,200.000000,15.000000" :Velocity%20Scale "1,1,1" :Size%20Scale%20%28Particles%29 1
          (:Out
            _)))
      (if (> (ScalableFloat|GetValueatLevel (Variables|BouncePad|GetSetIgnoreFallDamage)) 0.0)
        (Variables|BouncePad|SetIgnoreFallDamage true)
        (else
          (Variables|BouncePad|SetIgnoreFallDamage false))))))

(event EventOnFinishedBuilding
  ;[1 1 1] Finished building
  (Networking|CanExecuteCosmeticEvents
    (:True
      (if (not (Editor|InEditor))
        (Audio|PlaySoundatLocation (Variables|Audio|GetTrap_Active_Sound) (Transformation|GetActorLocation))))))

(event GameplayCue|GameplayCue.Abilities.Activation.Traps.ActivateTrap
  (params (EventType :type EGameplayCueEvent%20Enum :default "OnActive") (Parameters :type Gameplay%20Cue%20Parameters%20Structure%20%28by%20ref%29))
  (bind _overlappedactor (Variables|BouncePad|GetOverlappedActor))
  ;[1 1 1] GC for FX
  (switch Utilities|FlowControl|Switch|SwitchonEGameplayCueEvent EventType
    (:Executed
      (Utilities|Casting|CastToFortPlayerPawnAthena _overlappedactor
        (:then
          _)
        (:CastFailed
          (bind _asfort_athena_skvehicle (Utilities|Casting|CastToFortAthenaSKVehicle _overlappedactor)
            (:then
              (Effects|SpawnEmitterAttached :EmitterTemplate "/Game/Weapons/FORT_TRAPS/Trap_Jump_FullTile/Effects/P_Trap_Jump_FullTile_VehicleTrail.P_Trap_Jump_FullTile_VehicleTrail" :AttachToComponent (Variables|/Script/FortniteGame.FortAthenaSKVehicle|GetSkeletalMesh _asfort_athena_skvehicle) :Scale "1.000000,1.000000,1.000000" :LocationType "SnapToTarget" :bAutoDestroy true :PoolingMethod "None" :bAutoActivate true))
            (:CastFailed
              _))
          (Audio|PlaySoundatLocation (Variables|Audio|GetTrap_Fire_Sound) (Transformation|GetActorLocation))
          (Components|Activation|Activate (Variables|Default|GetBouncePadFX) true))))))

(event EventDestroyed
  ;[1 1 1] Trap Destroyed
  (if (not (Editor|InEditor))
    (Audio|PlaySoundatLocation (Variables|Audio|GetTrap_Destroyed_Sound) (Transformation|GetActorLocation))))

(event Collision|EventActorBeginOverlap
  (params (OtherActor :type Actor%20Object%20Reference))
  ;[1 1 1 0.5] Copies launch logic to ensure that overlapping worldstatic meshes don't block players from being launched (FORT-1103252).
  (if (not (Utilities|Array|ContainsItem (Variables|BouncePad|GetRecentlyLaunchedActors) OtherActor))
    (Utilities|FlowControl|Sequence
      (:then_0
        (CallFunction|HandleActorLaunch OtherActor)
        (GameplayCue|ExecuteGameplayCueOnActor(Burst) self (Variables|Default|GetActivationCueTag) (Ability|GameplayCue|MakeGameplayCueParameters :NormalizedMagnitude 0.0 :RawMagnitude 0.0 :Location "0,0,0" :Normal "0,0,0" :Instigator OtherActor :GameplayEffectLevel 1 :AbilityLevel 1 :bReplicateLocationWhenUsingMinimalRepProxy false)))
      (:then_1
        _)))
  (|CallParentFunction OtherActor)
  (Utilities|Time|SetTimerbyEvent (delegate Custom|ClearRecentlyLaunched) 0.15))

(event Custom|ClearRecentlyLaunched
  ;[1 1 1 0.5] Copies launch logic to ensure that overlapping worldstatic meshes don't block players from being launched (FORT-1103252).
  (Utilities|Array|Clear (Variables|BouncePad|GetRecentlyLaunchedActors)))

(fn OnTriggered
  (params (TouchingActors :type Array%20of%20Actor%20Object%20References))
  ;[1 1 1] On Overlap
  (Variables|BouncePad|SetRecentlyLaunchedActors TouchingActors)
  (for _array_element TouchingActors
    (CallFunction|HandleActorLaunch _array_element))
  (return false))

(fn FortAIPawnApplyGameplayEffects
  (bind _self self)
  (bind _returnvalue (Ability|GetAbilitySystemComponent _self))
  (bind _fortaipawn (Variables|Default|GetFortAIPawn))
  (bind _returnvalue_2 (Ability|GetAbilitySystemComponent _fortaipawn))
  (Variables|BouncePad|SetLaunchedVehicleInterface 0)
  (if (Variables|BouncePad|GetIgnoreFallDamage)
    (GameplayEffects|ApplyGameplayEffectToTarget :self _returnvalue :GameplayEffectClass (Variables|BouncePad|GetGE_LowGravity) :Target _returnvalue_2 :Level 0.0)
    (else
      ;[1 1 1] IgnoreFallDamage = FALSE
      (GameplayEffects|ApplyGameplayEffectToTarget :self _returnvalue :GameplayEffectClass (Variables|BouncePad|GetGE_NoFX_LowGravity) :Target _returnvalue_2 :Level 0.0)))
  (GameplayEffects|ApplyGameplayEffectToTarget :self _returnvalue :GameplayEffectClass (Variables|BouncePad|GetGE_BounceTracker1) :Target _returnvalue_2 :Level 0.0)
  (GameplayEffects|ApplyGameplayEffectToTarget :self _returnvalue :GameplayEffectClass (Variables|BouncePad|GetGE_BounceTracker2) :Target _returnvalue_2 :Level 0.0)
  (GameplayEffects|ApplyGameplayEffectToTarget :self _returnvalue :GameplayEffectClass (Variables|BouncePad|GetGE_LowGravityWithHopRocks) :Target _returnvalue_2 :Level 0.0)
  (Variables|Default|SetFortAIPawn 0))

(fn FortAIPawnLaunch
  (params (FortAIPawn :type Fort%20AIPawn%20Object%20Reference))
  (Character|LaunchCharacter FortAIPawn (+ (Math|Vector|ClampVectorSize (+ (* (Transformation|GetForwardVector (Variables|Trap|GetTraceLocation)) (ScalableFloat|GetValueatLevel (Variables|BouncePad|GetPlayerZVelocity))) (Math|Vector|ClampVectorSize (Math|Vector|MakeVector (.x (Math|Vector|MirrorVectorbyNormal (Transformation|GetVelocity FortAIPawn) (Transformation|GetForwardVector (Variables|Trap|GetTraceLocation)))) (.y (Math|Vector|MirrorVectorbyNormal (Transformation|GetVelocity FortAIPawn) (Transformation|GetForwardVector (Variables|Trap|GetTraceLocation))))) 1e+01 (ScalableFloat|GetValueatLevel (Variables|BouncePad|GetPlayerMaxLateralVelocity)))) (ScalableFloat|GetValueatLevel (Variables|BouncePad|GetPlayerMaxVelocity)) (ScalableFloat|GetValueatLevel (Variables|BouncePad|GetPlayerMaxVelocity))) "20.000000,0.000000, 0") true true)
  (Ability|SendGameplayEventtoActor FortAIPawn (Variables|Default|GetTag_CancelAbilityOnImpulse))
  (Ability|SendGameplayEventtoActor FortAIPawn (Variables|Default|GetTag_EventUsedLaunchPad))
  (CallFunction|FortAIPawnApplyGameplayEffects)
  (if (Variables|BouncePad|GetIgnoreFallDamage)
    (Variables|/Script/FortniteGame.FortPawn|SetbIgnoreNextFallingDamage true FortAIPawn)))

(fn HandleActorLaunch
  (params (OverlappingActor :type Actor%20Object%20Reference))
  ;[1 1 1 0.5] Handles launching actors (logic from "OnTriggered" spun off into blueprint-callable function - FORT-1103252)
  (bind _output_get (Variables|BouncePad|SetOverlappedActor OverlappingActor))
  (bind _asfort_player_pawn (Utilities|Casting|CastToFortPlayerPawn _output_get)
    (:then
      (Variables|BouncePad|SetLaunchedPlayer _asfort_player_pawn)
      (if (GameplayTags|HasMatchingGameplayTag _asfort_player_pawn (Variables|BouncePad|GetBouncePadTagMaxCount))
        (Actor|DestroyActor)
        (elif (Vehicle|IsinVehicle _asfort_player_pawn)
          (CallFunction|VehicleLaunch (Vehicle|GetVehicleInterface _asfort_player_pawn))
          (Effects|SpawnEmitterAttached "/Game/Weapons/FORT_TRAPS/Trap_Jump_FullTile/Effects/P_Trap_Jump_FullTile_PlayerTrail.P_Trap_Jump_FullTile_PlayerTrail" (Variables|/Script/Engine.Character|GetMesh _asfort_player_pawn) "root") ; Bypassing because FX is on player
          (else
            (CallFunction|PlayerLaunch _asfort_player_pawn)))))
    (:CastFailed
      (bind _asfort_vehicle_interface (Utilities|Casting|CastToFortVehicleInterface (Variables|BouncePad|GetOverlappedActor))
        (:then
          (Variables|BouncePad|SetLaunchedVehicleInterface _asfort_vehicle_interface)
          (CallFunction|VehicleLaunch _asfort_vehicle_interface))
        (:CastFailed
          (bind _asfort_aipawn (Utilities|Casting|CastToFortAIPawn (Variables|BouncePad|GetOverlappedActor))
            (:then
              (bind _output_get_2 (Variables|Default|SetFortAIPawn _asfort_aipawn))
              (CallFunction|FortAIPawnLaunch _output_get_2))
            (:CastFailed
              (bind _asfort_physics_pawn_object (Utilities|Casting|CastToFortPhysicsPawnObject (Variables|BouncePad|GetOverlappedActor))
                (:then
                  (CallFunction|PhysicsObjectLaunch _asfort_physics_pawn_object))
                (:CastFailed
                  (CallFunction|PhysicsPropLaunch (Variables|BouncePad|GetOverlappedActor)))))))))))

(fn OnRep_OverlappedActor)

(fn PhysicsObjectLaunch
  (params (PhysicsObject :type Fort%20Physics%20Pawn%20Object%20Object%20Reference))
  (bind _physicsmeshcomponent (Variables|/Script/FortniteGame.FortPhysicsPawnObject|GetPhysicsMeshComponent PhysicsObject))
  (bind _returnvalue (Physics|GetPhysicsLinearVelocity _physicsmeshcomponent "None"))
  (Physics|SetPhysicsLinearVelocity _physicsmeshcomponent (+ (Math|Vector|ClampVectorSize (+ (* (Transformation|GetForwardVector (Variables|Trap|GetTraceLocation)) (ScalableFloat|GetValueatLevel (Variables|BouncePad|GetPlayerZVelocity))) (Math|Vector|ClampVectorSize (Math|Vector|MakeVector (.x (Math|Vector|MirrorVectorbyNormal _returnvalue (Transformation|GetForwardVector (Variables|Trap|GetTraceLocation)))) (.y (Math|Vector|MirrorVectorbyNormal _returnvalue (Transformation|GetForwardVector (Variables|Trap|GetTraceLocation))))) 1e+01 (ScalableFloat|GetValueatLevel (Variables|BouncePad|GetPlayerMaxLateralVelocity)))) (ScalableFloat|GetValueatLevel (Variables|BouncePad|GetPlayerMaxVelocity)) (ScalableFloat|GetValueatLevel (Variables|BouncePad|GetPlayerMaxVelocity))) "20.000000,0.000000, 0") false "None"))

(fn PhysicsPropLaunch
  (params (PhysicsPropToLaunch :type Actor%20Object%20Reference))
  (Class|NewtonPhysicsPropsLibrary|LaunchPhysicsProps self PhysicsPropToLaunch (Transformation|GetActorUpVector) (Variables|Default|GetMinPhysicsPropMass) (Variables|Default|GetMaxPhysicsPropMass) (Variables|Default|GetMinPhysicsPropImpulse) (Variables|Default|GetMaxPhysicsPropImpulse))
  (return))

(fn PlayerApplyGameplayEffects
  (bind _self self)
  (bind _returnvalue (Ability|GetAbilitySystemComponent _self))
  (bind _launchedplayer (Variables|BouncePad|GetLaunchedPlayer))
  (bind _returnvalue_2 (Ability|GetAbilitySystemComponent _launchedplayer))
  (Variables|BouncePad|SetLaunchedVehicleInterface 0)
  (if (Variables|BouncePad|GetIgnoreFallDamage)
    (GameplayEffects|ApplyGameplayEffectToTarget :self _returnvalue :GameplayEffectClass (Variables|BouncePad|GetGE_LowGravity) :Target _returnvalue_2 :Level 0.0)
    (else
      ;[1 1 1] IgnoreFallDamage = FALSE
      (GameplayEffects|ApplyGameplayEffectToTarget :self _returnvalue :GameplayEffectClass (Variables|BouncePad|GetGE_NoFX_LowGravity) :Target _returnvalue_2 :Level 0.0)))
  (GameplayEffects|ApplyGameplayEffectToTarget :self _returnvalue :GameplayEffectClass (Variables|BouncePad|GetGE_BounceTracker1) :Target _returnvalue_2 :Level 0.0)
  (GameplayEffects|ApplyGameplayEffectToTarget :self _returnvalue :GameplayEffectClass (Variables|BouncePad|GetGE_BounceTracker2) :Target _returnvalue_2 :Level 0.0)
  (GameplayEffects|ApplyGameplayEffectToTarget :self _returnvalue :GameplayEffectClass (Variables|BouncePad|GetGE_LowGravityWithHopRocks) :Target _returnvalue_2 :Level 0.0)
  (Variables|BouncePad|SetLaunchedPlayer 0))

(fn PlayerLaunch
  (params (Player :type Fort%20Player%20Pawn%20Object%20Reference))
  (Character|LaunchCharacter Player (Math|Vector|ClampVectorSize (+ (* (Transformation|GetForwardVector (Variables|Trap|GetTraceLocation)) (ScalableFloat|GetValueatLevel (Variables|BouncePad|GetPlayerZVelocity))) (Math|Vector|ClampVectorSize (Math|Vector|MakeVector (.x (Math|Vector|MirrorVectorbyNormal (Transformation|GetVelocity Player) (Transformation|GetForwardVector (Variables|Trap|GetTraceLocation)))) (.y (Math|Vector|MirrorVectorbyNormal (Transformation|GetVelocity Player) (Transformation|GetForwardVector (Variables|Trap|GetTraceLocation))))) 0.0 (ScalableFloat|GetValueatLevel (Variables|BouncePad|GetPlayerMaxLateralVelocity)))) (ScalableFloat|GetValueatLevel (Variables|BouncePad|GetPlayerMaxVelocity)) (ScalableFloat|GetValueatLevel (Variables|BouncePad|GetPlayerMaxVelocity))) true true)
  (Ability|SendGameplayEventtoActor Player (Variables|Default|GetTag_CancelAbilityOnImpulse))
  (Ability|SendGameplayEventtoActor Player "(TagName=\"Granted.Athena.EnvItem.Bouncer.Launch\")" (Utilities|Struct|MakeGameplayEventData "(TagName=\"Granted.Athena.EnvItem.Bouncer.Launch\")" 0 Player 0 0 "()" "(GameplayTags=)" (GameplayTags|GetOwnedGameplayTags Player :!class /Script/GameplayTags.GameplayTagAssetInterface)))
  (CallFunction|PlayerApplyGameplayEffects)
  (if (Variables|BouncePad|GetIgnoreFallDamage)
    (Variables|/Script/FortniteGame.FortPawn|SetbIgnoreNextFallingDamage true Player)))

(fn SendAntelopLaunchEvent
  (params (Players :type Array%20of%20Fort%20Player%20Pawn%20Object%20References))
  (for _array_element Players
    (if (Networking|IsServer)
      (Quest|SendComplexCustomStatEvent (Quests|GetQuestManager (Fortnite|GetFortPlayerControllerfromActor _array_element) "Invalid") 0 (GameplayTags|MakeGameplayTagContainerfromTag (LocalVariables|GetLaunchAntelopStatEvent)) "(GameplayTags=)" 1))))

(fn ConstructionScript
  (Collision|SetCollisionEnabled (Variables|BouncePad|GetTrigger) "NoCollision"))

(fn VehicleLaunch
  (params (VehicleInterface :type Fort%20Vehicle%20Interface%20Interface))
  (bind _returnvalue (Vehicle|GetAllPassengers VehicleInterface))
  (bind _returnvalue_2 (Vehicle|GetVehicleActor VehicleInterface :!class /Script/FortniteGame.FortVehicleInterface))
  (bind _asskeletal_mesh_component (Utilities|Casting|CastToSkeletalMeshComponent (Variables|/Script/Engine.Actor|GetRootComponent _returnvalue_2)))
  (Audio|BroadcastAbilityImpactatLocation (Variables|Audio|GetTrap_Active_Sound) _returnvalue_2 (Math|Vector|ClampVectorSize (+ (Math|Vector|MakeVector (.x (+ (+ (Math|Vector|MirrorVectorbyNormal (Vehicle|GetVehicleVelocity VehicleInterface) (Transformation|GetForwardVector (Variables|Trap|GetTraceLocation))) (* (Transformation|GetForwardVector (Variables|Trap|GetTraceLocation)) (* (Physics|GetMass _asskeletal_mesh_component) (ScalableFloat|GetValueatLevel (Variables|BouncePad|GetVehicleMassVelocityMultiplier))))) "20.000000,0.000000, 0")) (.y (+ (+ (Math|Vector|MirrorVectorbyNormal (Vehicle|GetVehicleVelocity VehicleInterface) (Transformation|GetForwardVector (Variables|Trap|GetTraceLocation))) (* (Transformation|GetForwardVector (Variables|Trap|GetTraceLocation)) (* (Physics|GetMass _asskeletal_mesh_component) (ScalableFloat|GetValueatLevel (Variables|BouncePad|GetVehicleMassVelocityMultiplier))))) "20.000000,0.000000, 0"))) (* (Transformation|GetForwardVector (Variables|Trap|GetTraceLocation)) (ScalableFloat|GetValueatLevel (Variables|BouncePad|GetVehicleMinLateralVelocity)))) (ScalableFloat|GetValueatLevel (Variables|BouncePad|GetVehicleMinLateralVelocity)) (ScalableFloat|GetValueatLevel (Variables|BouncePad|GetVehicleMaxLateralVelocity))) 7.8e+02) ; Set Physics Linear Velocity
  (if (Variables|BouncePad|GetIgnoreFallDamage)
    (Vehicle|SetIgnoreNextFallingDamage VehicleInterface true)
    (if (GameplayTags|HasTag (Class|FortVehicleInterface|GetVehicleTags VehicleInterface) "(TagName=\"Vehicle.Antelope\")" true)
      (CallFunction|SendAntelopLaunchEvent _returnvalue))
    (for _array_element _returnvalue
      (Effects|SpawnEmitterAttached "/Game/Weapons/FORT_TRAPS/Trap_Jump_FullTile/Effects/P_Trap_Jump_FullTile_PlayerTrail.P_Trap_Jump_FullTile_PlayerTrail" (Variables|/Script/Engine.Character|GetMesh _array_element) "root") ; Bypassing because FX is on player
      (Variables|BouncePad|SetLaunchedPlayer _array_element)
      (CallFunction|PlayerApplyGameplayEffects)
      (Variables|/Script/FortniteGame.FortPawn|SetbIgnoreNextFallingDamage true _array_element))))
