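; Wall-mounted bounce pad trap that launches players, vehicles, AI pawns, physics objects and props away from the wall.
; OnTriggered dispatches by actor type, with launch velocities mostly mirrored or pushed along the trace forward vector and clamped by scalable floats.
; Players and AI receive low-gravity and bounce-tracker effects, with optional fall-damage immunity; a player carrying the max-count tag destroys the trap instead.
; Placement events register the trigger; a GameplayCue handles firing FX.
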(event EventBeginPlay
  (|CallParentFunction)
  (Collision|SetCollisionEnabled (Variables|BouncePad|GetTrigger) "QueryAndPhysics"))

(event EditorPreview|EventSetupForEditorPreview
  ;[0.5917 1 0 0.5] Ensure the trigger is set up when placed in UEFN during VK-Edit
  (|CallParentFunction)
  (Utilities|FlowControl|DoOnce false
    (:Completed
      (Utilities|FlowControl|Sequence
        (:then_0
          (Collision|SetCollisionEnabled (Variables|BouncePad|GetTrigger) "QueryAndPhysics"))
        (:then_1
          ;[1 1 1] Initial Setup
          (Utilities|FlowControl|DoOnce false
            (:Completed
              (Trap|AddTriggerComponent (Variables|BouncePad|GetTrigger))
              (if (Editor|InEditor)
                _
                (elif (Variables|Default|GetbPlayFewerSounds)
                  (Utilities|TrapPlacementDustFX :Attach%20to%20Component (Variables|Debug|GetStaticMeshComponent) :Socket "PlacedFX" :Rotation "0, 0, 0" :Burst%20Rate 15 :Size%20%28Volume%29 "200.000000,200.000000,15.000000" :Velocity%20Scale "1,1,1" :Size%20Scale%20%28Particles%29 1
                    (:Out
                      _))
                  (else
                    (Audio|PlaySoundatLocation (Variables|Audio|GetPlace_Trap_Sound) (Transformation|GetActorLocation)))))
              (if (> (ScalableFloat|GetValueatLevel (Variables|BouncePad|GetSetIgnoreFallDamage)) 0.0)
                (Variables|BouncePad|SetIgnoreFallDamage true)
                (else
                  (Variables|BouncePad|SetIgnoreFallDamage false))))))))) ; Otherwise this happens whenever it's moved in the editor
)

(event EventOnPlaced
  ;[1 1 1] Initial Setup
  (Utilities|FlowControl|DoOnce false
    (:Completed
      (Trap|AddTriggerComponent (Variables|BouncePad|GetTrigger))
      (if (Editor|InEditor)
        _
        (elif (Variables|Default|GetbPlayFewerSounds)
          (Utilities|TrapPlacementDustFX :Attach%20to%20Component (Variables|Debug|GetStaticMeshComponent) :Socket "PlacedFX" :Rotation "0, 0, 0" :Burst%20Rate 15 :Size%20%28Volume%29 "200.000000,200.000000,15.000000" :Velocity%20Scale "1,1,1" :Size%20Scale%20%28Particles%29 1
            (:Out
              _))
          (else
            (Audio|PlaySoundatLocation (Variables|Audio|GetPlace_Trap_Sound) (Transformation|GetActorLocation)))))
      (if (> (ScalableFloat|GetValueatLevel (Variables|BouncePad|GetSetIgnoreFallDamage)) 0.0)
        (Variables|BouncePad|SetIgnoreFallDamage true)
        (else
          (Variables|BouncePad|SetIgnoreFallDamage false))))))

(event EventOnFinishedBuilding
  ;[1 1 1] Finished building
  (Networking|CanExecuteCosmeticEvents
    (:True
      (if (not (or (Editor|InEditor) (Variables|Default|GetbPlayFewerSounds)))
        (Audio|PlaySoundatLocation (Variables|Audio|GetTrap_Active_Sound) (Transformation|GetActorLocation))))))

(event GameplayCue|GameplayCue.Abilities.Activation.Traps.ActivateTrap
  (params (EventType :type EGameplayCueEvent%20Enum :default "OnActive") (Parameters :type Gameplay%20Cue%20Parameters%20Structure%20%28by%20ref%29))
  (bind _overlappedactor (Variables|BouncePad|GetOverlappedActor))
  ;[1 1 1] GC Triggered on Overlap
  (switch Utilities|FlowControl|Switch|SwitchonEGameplayCueEvent EventType
    (:Executed
      (Utilities|Casting|CastToFortPlayerPawnAthena _overlappedactor
        (:then
          _)
        (:CastFailed
          (bind _asfort_athena_skvehicle (Utilities|Casting|CastToFortAthenaSKVehicle _overlappedactor)
            (:then
              (Effects|SpawnEmitterAttached "/Game/Weapons/FORT_TRAPS/Trap_Jump_FullTile/Effects/P_Trap_Jump_FullTile_VehicleTrail.P_Trap_Jump_FullTile_VehicleTrail" (Variables|/Script/FortniteGame.FortAthenaSKVehicle|GetSkeletalMesh _asfort_athena_skvehicle) ""))
            (:CastFailed
              _))
          (Audio|PlaySoundatLocation (Variables|Audio|GetTrap_Fire_Sound) (Transformation|GetActorLocation))
          (Components|Activation|Activate (Variables|Default|GetBouncePadFX) true))))))

(event EventDestroyed
  ;[1 1 1] Trap Destroyed
  (if (not (Editor|InEditor))
    (Audio|PlaySoundatLocation (Variables|Audio|GetTrap_Destroyed_Sound) (Transformation|GetActorLocation))))

(fn OnTriggered
  (params (TouchingActors :type Array%20of%20Actor%20Object%20References))
  (bind _tracelocation (Variables|Trap|GetTraceLocation))
  (bind _returnvalue (Transformation|GetForwardVector _tracelocation))
  ;[1 1 1] On Overlap
  (for _array_element TouchingActors
    (bind _output_get (Variables|BouncePad|SetOverlappedActor _array_element))
    (bind _asfort_player_pawn (Utilities|Casting|CastToFortPlayerPawn _output_get)
      (:then
        (bind _output_get_2 (Variables|BouncePad|SetLaunchedPlayer _asfort_player_pawn))
        (if (GameplayTags|HasMatchingGameplayTag _output_get_2 (Variables|BouncePad|GetBouncePadTagMaxCount))
          (Actor|DestroyActor)
          (elif (Vehicle|IsinVehicle _asfort_player_pawn)
            (CallFunction|VehicleLaunch (Vehicle|GetVehicleInterface _asfort_player_pawn))
            (Effects|SpawnEmitterAttached "/Game/Weapons/FORT_TRAPS/Trap_Jump_FullTile/Effects/P_Trap_Jump_FullTile_PlayerTrail.P_Trap_Jump_FullTile_PlayerTrail" (Variables|/Script/Engine.Character|GetMesh _asfort_player_pawn) "root") ; Bypassing because FX is on player
            (elif (> (Math|Vector|DotProduct _returnvalue (Transformation|GetVelocity _asfort_player_pawn)) -2e+02)
              (CallFunction|PlayerLaunch _asfort_player_pawn (+ (Transformation|GetVelocity _asfort_player_pawn) (* (Transformation|GetForwardVector (Variables|Trap|GetTraceLocation)) (ScalableFloat|GetValueatLevel (Variables|BouncePad|GetPlayerMinLateralVelocity)))))
              (else
                (CallFunction|PlayerLaunch _asfort_player_pawn (Math|Vector|MirrorVectorbyNormal (Transformation|GetVelocity _asfort_player_pawn) _returnvalue)))))))
      (:CastFailed
        ;;[1 1 1] Vehicle Launch
        (bind _asfort_athena_vehicle (Utilities|Casting|CastToFortAthenaVehicle (Variables|BouncePad|GetOverlappedActor))
          (:then
            (bind _output_get_3 (Variables|BouncePad|SetLaunchedVehicle _asfort_athena_vehicle))
            (CallFunction|VehicleLaunch _output_get_3))
          (:CastFailed
            (bind _asfort_aipawn (Utilities|Casting|CastToFortAIPawn (Variables|BouncePad|GetOverlappedActor))
              (:then
                (bind _output_get_4 (Variables|BouncePad|SetFortAIPawn _asfort_aipawn))
                (CallFunction|FortAIPawnLaunch _output_get_4))
              (:CastFailed
                (bind _asfort_physics_pawn_object (Utilities|Casting|CastToFortPhysicsPawnObject (Variables|BouncePad|GetOverlappedActor))
                  (:then
                    (CallFunction|PhysicsObjectLaunch _asfort_physics_pawn_object))
                  (:CastFailed
                    (CallFunction|PhysicsPropLaunch (Variables|BouncePad|GetOverlappedActor)))))))))))
  (return false))

(fn FortAIPawnApplyGameplayEffects
  (bind _self self)
  (bind _returnvalue (Ability|GetAbilitySystemComponent _self))
  (bind _fortaipawn (Variables|BouncePad|GetFortAIPawn))
  (bind _returnvalue_2 (Ability|GetAbilitySystemComponent _fortaipawn))
  (Variables|BouncePad|SetLaunchedVehicle 0)
  (if (Variables|BouncePad|GetIgnoreFallDamage)
    (GameplayEffects|ApplyGameplayEffectToTarget :self _returnvalue :GameplayEffectClass (Variables|BouncePad|GetGE_LowGravity) :Target _returnvalue_2 :Level 0.0)
    (else
      ;[1 1 1] IgnoreFallDamage = FALSE
      (GameplayEffects|ApplyGameplayEffectToTarget :self _returnvalue :GameplayEffectClass (Variables|BouncePad|GetGE_NoFXLowGravity) :Target _returnvalue_2 :Level 0.0)))
  (GameplayEffects|ApplyGameplayEffectToTarget :self _returnvalue :GameplayEffectClass (Variables|BouncePad|GetGE_BounceTracker1) :Target _returnvalue_2 :Level 0.0)
  (GameplayEffects|ApplyGameplayEffectToTarget :self _returnvalue :GameplayEffectClass (Variables|BouncePad|GetGE_BounceTracker2) :Target _returnvalue_2 :Level 0.0)
  (GameplayEffects|ApplyGameplayEffectToTarget :self _returnvalue :GameplayEffectClass (Variables|BouncePad|GetGE_LowGravityWithHopRocks) :Target _returnvalue_2 :Level 0.0)
  (Variables|BouncePad|SetFortAIPawn 0))

(fn FortAIPawnLaunch
  (params (Pawn :type Fort%20Pawn%20Object%20Reference))
  (Character|LaunchCharacter Pawn (Math|Vector|ClampVectorSize (+ (Math|Vector|MakeVector (.x (+ (Math|Vector|MirrorVectorbyNormal (Transformation|GetVelocity Pawn) (Transformation|GetForwardVector (Variables|Trap|GetTraceLocation))) (* (Transformation|GetForwardVector (Variables|Trap|GetTraceLocation)) (ScalableFloat|GetValueatLevel (Variables|BouncePad|GetPlayerMaxLateralVelocity))))) (.y (+ (Math|Vector|MirrorVectorbyNormal (Transformation|GetVelocity Pawn) (Transformation|GetForwardVector (Variables|Trap|GetTraceLocation))) (* (Transformation|GetForwardVector (Variables|Trap|GetTraceLocation)) (ScalableFloat|GetValueatLevel (Variables|BouncePad|GetPlayerMaxLateralVelocity))))) (ScalableFloat|GetValueatLevel (Variables|BouncePad|GetPlayerMinZVelocity))) (* (Transformation|GetForwardVector (Variables|Trap|GetTraceLocation)) (ScalableFloat|GetValueatLevel (Variables|BouncePad|GetPlayerMinLateralVelocity)))) (ScalableFloat|GetValueatLevel (Variables|BouncePad|GetPlayerMinLateralVelocity)) (ScalableFloat|GetValueatLevel (Variables|BouncePad|GetPlayerMaxLateralVelocity))) true true)
  (Ability|SendGameplayEventtoActor Pawn (Variables|Default|GetTag_CancelAbilityOnImpulse))
  (CallFunction|FortAIPawnApplyGameplayEffects))

(fn OnRep_Overlappedactor)

(fn PhysicsObjectLaunch
  (params (PhysicsObject :type Fort%20Physics%20Pawn%20Object%20Object%20Reference))
  (bind _physicsmeshcomponent (Variables|/Script/FortniteGame.FortPhysicsPawnObject|GetPhysicsMeshComponent PhysicsObject))
  (bind _returnvalue (Physics|GetPhysicsLinearVelocity _physicsmeshcomponent "None"))
  (Physics|SetPhysicsLinearVelocity _physicsmeshcomponent (+ (Math|Vector|ClampVectorSize (+ (* (Transformation|GetForwardVector (Variables|Trap|GetTraceLocation)) (ScalableFloat|GetValueatLevel (Variables|BouncePad|GetPlayerMinLateralVelocity))) (Math|Vector|ClampVectorSize (Math|Vector|MakeVector (.x (Math|Vector|MirrorVectorbyNormal _returnvalue (Transformation|GetForwardVector (Variables|Trap|GetTraceLocation)))) (.y (Math|Vector|MirrorVectorbyNormal _returnvalue (Transformation|GetForwardVector (Variables|Trap|GetTraceLocation))))) 1e+01 (ScalableFloat|GetValueatLevel (Variables|BouncePad|GetPlayerMaxLateralVelocity)))) (ScalableFloat|GetValueatLevel (Variables|BouncePad|GetPlayerMaxLateralVelocity)) (ScalableFloat|GetValueatLevel (Variables|BouncePad|GetPlayerMaxLateralVelocity))) "20.000000,0.000000, 0") false "None"))

(fn PhysicsPropLaunch
  (params (PhysicsPropToLaunch :type Actor%20Object%20Reference))
  (Class|NewtonPhysicsPropsLibrary|LaunchPhysicsProps self PhysicsPropToLaunch (Transformation|GetForwardVector (Variables|Trap|GetTraceLocation)) (Variables|Default|GetMinPhysicsPropMass) (Variables|Default|GetMaxPhysicsPropMass) (Variables|Default|GetMinPhysicsPropImpulse) (Variables|Default|GetMaxPhysicsPropImpulse))
  (return))

(fn PlayerApplyGameplayEffects
  (bind _self self)
  (bind _returnvalue (Ability|GetAbilitySystemComponent _self))
  (bind _launchedplayer (Variables|BouncePad|GetLaunchedPlayer))
  (bind _returnvalue_2 (Ability|GetAbilitySystemComponent _launchedplayer))
  (Variables|BouncePad|SetLaunchedVehicle 0)
  (if (Variables|BouncePad|GetIgnoreFallDamage)
    (GameplayEffects|ApplyGameplayEffectToTarget :self _returnvalue :GameplayEffectClass (Variables|BouncePad|GetGE_LowGravity) :Target _returnvalue_2 :Level 0.0)
    (else
      ;[1 1 1] IgnoreFallDamage = FALSE
      (GameplayEffects|ApplyGameplayEffectToTarget :self _returnvalue :GameplayEffectClass (Variables|BouncePad|GetGE_NoFXLowGravity) :Target _returnvalue_2 :Level 0.0)))
  (GameplayEffects|ApplyGameplayEffectToTarget :self _returnvalue :GameplayEffectClass (Variables|BouncePad|GetGE_BounceTracker1) :Target _returnvalue_2 :Level 0.0)
  (GameplayEffects|ApplyGameplayEffectToTarget :self _returnvalue :GameplayEffectClass (Variables|BouncePad|GetGE_BounceTracker2) :Target _returnvalue_2 :Level 0.0)
  (GameplayEffects|ApplyGameplayEffectToTarget :self _returnvalue :GameplayEffectClass (Variables|BouncePad|GetGE_LowGravityWithHopRocks) :Target _returnvalue_2 :Level 0.0)
  (Variables|BouncePad|SetLaunchedPlayer 0))

(fn PlayerLaunch
  (params (FortPlayerPawn :type Fort%20Player%20Pawn%20Object%20Reference) (InVec :type Vector))
  (bind _playerminlateralvelocity (Variables|BouncePad|GetPlayerMinLateralVelocity))
  (bind _returnvalue (ScalableFloat|GetValueatLevel _playerminlateralvelocity))
  (bind _playermaxlateralvelocity (Variables|BouncePad|GetPlayerMaxLateralVelocity))
  (bind _returnvalue_2 (ScalableFloat|GetValueatLevel _playermaxlateralvelocity))
  (bind _playerminzvelocity (Variables|BouncePad|GetPlayerMinZVelocity))
  (bind _returnvalue_3 (ScalableFloat|GetValueatLevel _playerminzvelocity))
  (bind _playervelocitymultiplier (Variables|BouncePad|GetPlayerVelocityMultiplier))
  (bind _returnvalue_4 (ScalableFloat|GetValueatLevel _playervelocitymultiplier))
  (bind _returnvalue_5 (* InVec _returnvalue_4))
  (bind _x (.x _returnvalue_5))
  (bind _y (.y _returnvalue_5))
  (bind _z (.z _returnvalue_5))
  (bind _returnvalue_6 (Math|Vector|MakeVector _x _y))
  (bind _returnvalue_7 (Math|Vector|ClampVectorSize _returnvalue_6 _returnvalue _returnvalue_2))
  (bind _x_2 (.x _returnvalue_7))
  (bind _y_2 (.y _returnvalue_7))
  (bind _playermaxzvelocity (Variables|BouncePad|GetPlayerMaxZVelocity))
  (bind _returnvalue_8 (ScalableFloat|GetValueatLevel _playermaxzvelocity))
  (bind _returnvalue_9 (Math|Float|Clamp(Float) _z _returnvalue_3 _returnvalue_8))
  (bind _returnvalue_10 (Math|Vector|MakeVector _x_2 _y_2 _returnvalue_9))
  (if (<= (Math|Vector|VectorLength _returnvalue_10) (+ (ScalableFloat|GetValueatLevel (Variables|BouncePad|GetPlayerMinZVelocity)) 5e+01))
    (Character|LaunchCharacter FortPlayerPawn (+ (* (Transformation|GetForwardVector (Variables|Trap|GetTraceLocation)) (ScalableFloat|GetValueatLevel (Variables|BouncePad|GetPlayerMinLateralVelocity))) "0.000000,0.000000,500.000000") true true)
    (else
      (Character|LaunchCharacter FortPlayerPawn _returnvalue_10 true true)))
  (Ability|SendGameplayEventtoActor FortPlayerPawn (Variables|Default|GetTag_CancelAbilityOnImpulse))
  (CallFunction|PlayerApplyGameplayEffects)
  (if (Variables|BouncePad|GetIgnoreFallDamage)
    (Variables|/Script/FortniteGame.FortPawn|SetbIgnoreNextFallingDamage true FortPlayerPawn)))

(fn ConstructionScript
  (Collision|SetCollisionEnabled (Variables|BouncePad|GetTrigger) "NoCollision"))

(fn VehicleLaunch
  (params (VehicleInterface :type Fort%20Vehicle%20Interface%20Interface))
  (bind _returnvalue (Vehicle|GetVehicleActor VehicleInterface :!class /Script/FortniteGame.FortVehicleInterface))
  (bind _asskeletal_mesh_component (Utilities|Casting|CastToSkeletalMeshComponent (Variables|/Script/Engine.Actor|GetRootComponent _returnvalue)))
  (Audio|BroadcastAbilityImpactatLocation (Variables|Audio|GetTrap_Active_Sound) _returnvalue (Math|Vector|ClampVectorSize (+ (Math|Vector|MakeVector (.x (+ (Math|Vector|MirrorVectorbyNormal (Vehicle|GetVehicleVelocity VehicleInterface) (Transformation|GetForwardVector (Variables|Trap|GetTraceLocation))) (* (Transformation|GetForwardVector (Variables|Trap|GetTraceLocation)) (* (Physics|GetMass _asskeletal_mesh_component) (ScalableFloat|GetValueatLevel (Variables|BouncePad|GetVehicleVelocityMultiplier)))))) (.y (+ (Math|Vector|MirrorVectorbyNormal (Vehicle|GetVehicleVelocity VehicleInterface) (Transformation|GetForwardVector (Variables|Trap|GetTraceLocation))) (* (Transformation|GetForwardVector (Variables|Trap|GetTraceLocation)) (* (Physics|GetMass _asskeletal_mesh_component) (ScalableFloat|GetValueatLevel (Variables|BouncePad|GetVehicleVelocityMultiplier)))))) (ScalableFloat|GetValueatLevel (Variables|BouncePad|GetVehicleZVelocity))) (* (Transformation|GetForwardVector (Variables|Trap|GetTraceLocation)) (ScalableFloat|GetValueatLevel (Variables|BouncePad|GetVehicleMinLateralVelocity)))) (ScalableFloat|GetValueatLevel (Variables|BouncePad|GetVehicleMinLateralVelocity)) (ScalableFloat|GetValueatLevel (Variables|BouncePad|GetVehicleMaxVelocity))) 7.8e+02) ; Set Physics Linear Velocity
  (if (Variables|BouncePad|GetIgnoreFallDamage)
    (Vehicle|SetIgnoreNextFallingDamage VehicleInterface true)
    (for _array_element (Vehicle|GetAllPassengers VehicleInterface)
      (Effects|SpawnEmitterAttached "/Game/Weapons/FORT_TRAPS/Trap_Jump_FullTile/Effects/P_Trap_Jump_FullTile_PlayerTrail.P_Trap_Jump_FullTile_PlayerTrail" (Variables|/Script/Engine.Character|GetMesh _array_element) "root") ; Bypassing because FX is on player
      (Variables|BouncePad|SetLaunchedPlayer _array_element)
      (CallFunction|PlayerApplyGameplayEffects)
      (Variables|/Script/FortniteGame.FortPawn|SetbIgnoreNextFallingDamage true _array_element))))
